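# Animation Blueprint for the FreightCalf pickaxe, playing its weapon equip animation and adding secondary physics motion to its wire and blade bones.
# Each update sets Equipped from a check for the Style.OptionA meta tag on the equipped pickaxe, switching the state machine between the one-shot equip animation and reference pose.
# That cached pose feeds chained AnimDynamics nodes for left and right wires and two blade parts, whose alpha follows the DynamicsAlpha curve.
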AnimBlueprint /BRCosmetics/Weapons/FORT_Melee/Pickaxe_FreightCalf/Meshes/FreightCalf_Axe_AnimBP
target_skeleton=&/FreightCalf_Axe_Skeleton
graphs [
 AnimGraph {nodes [
   n1 = Anim Dynamics - Bone: dyn_wire_l_1(ComponentPose=Local To Component(LocalPose=Use cached pose 'Input pose cache'()), Alpha=1, ExternalForce="0,0,0") {
    angularBiasOverride=0.999
    angularDampingOverride=0.9999
    angularSpringConstant=200000
    bOverrideAngularDamping=true
    simulationSpace=BoneRelative
    alphaBoolBlend {blendOption=Linear}
    boundBone {boneName=dyn_wire_l_1}
    componentAppliedLinearAccClamp=<2000 2000 2000>
    componentLinearAccScale=<0.3 0.3 0.3>
    componentLinearVelScale=<0.3 0.3 0.3>
    physicsBodyDefinitions [{
      boundBone {boneName=dyn_wire_l_1}
      boxExtents=<30 30 30>
      constraintSetup {angularLimitsMax=<5 2 10>, angularLimitsMin=<-5 -2 -10>, angularTarget=<0 0 0>}
      localJointOffset=<10 0 0>
     }]
    relativeSpaceBone {boneName=Melee_TwoHanded_Handle}
   }
   n2 = Anim Dynamics - Bone: dyn_wire_l_2(ComponentPose=@n1, Alpha=1, ExternalForce="0,0,0") {
    angularBiasOverride=0.999
    angularDampingOverride=0.999
    angularSpringConstant=200000
    bAngularSpring=true
    bOverrideAngularDamping=true
    simulationSpace=BoneRelative
    alphaBoolBlend {blendOption=Linear}
    boundBone {boneName=dyn_wire_l_2}
    componentAppliedLinearAccClamp=<2000 2000 2000>
    componentLinearAccScale=<0.2 0.2 0.2>
    componentLinearVelScale=<0.2 0.2 0.2>
    physicsBodyDefinitions [{
      boundBone {boneName=dyn_wire_l_2}
      boxExtents=<25 25 25>
      constraintSetup {angularLimitsMax=<5 5 2>, angularLimitsMin=<-5 -5 -2>, angularTarget=<0 0 0>}
      localJointOffset=<10 0 0>
     }]
    relativeSpaceBone {boneName=Melee_TwoHanded_Handle}
   }
   n3 = Anim Dynamics - Bone: dyn_wire_r_1(ComponentPose=@n2, Alpha=1, ExternalForce="0,0,0") {
    angularBiasOverride=0.999
    angularDampingOverride=0.9999
    angularSpringConstant=200000
    bOverrideAngularDamping=true
    simulationSpace=BoneRelative
    alphaBoolBlend {blendOption=Linear}
    boundBone {boneName=dyn_wire_r_1}
    componentAppliedLinearAccClamp=<2000 2000 2000>
    componentLinearAccScale=<0.3 0.3 0.3>
    componentLinearVelScale=<0.3 0.3 0.3>
    physicsBodyDefinitions [{
      boundBone {boneName=dyn_wire_r_1}
      boxExtents=<30 30 30>
      constraintSetup {angularLimitsMax=<5 2 10>, angularLimitsMin=<-5 -2 -10>, angularTarget=<0 0 0>}
      localJointOffset=<10 0 0>
     }]
    relativeSpaceBone {boneName=Melee_TwoHanded_Handle}
   }
   n4 = Anim Dynamics - Bone: dyn_wire_r_2(ComponentPose=@n3, Alpha=1, ExternalForce="0,0,0") {
    angularBiasOverride=0.999
    angularDampingOverride=0.999
    angularSpringConstant=200000
    bAngularSpring=true
    bOverrideAngularDamping=true
    simulationSpace=BoneRelative
    alphaBoolBlend {blendOption=Linear}
    boundBone {boneName=dyn_wire_r_2}
    componentAppliedLinearAccClamp=<2000 2000 2000>
    componentLinearAccScale=<0.2 0.2 0.2>
    componentLinearVelScale=<0.2 0.2 0.2>
    physicsBodyDefinitions [{
      boundBone {boneName=dyn_wire_r_2}
      boxExtents=<25 25 25>
      constraintSetup {angularLimitsMax=<5 5 2>, angularLimitsMin=<-5 -5 -2>, angularTarget=<0 0 0>}
      localJointOffset=<10 0 0>
     }]
    relativeSpaceBone {boneName=Melee_TwoHanded_Handle}
   }
   n5 = Anim Dynamics - Bone: dyn_blade_part_2!(ComponentPose=@n4, AlphaCurveName:Name=DynamicsAlpha, ExternalForce="0,0,0") -> (Pose) {
    alphaInputType=Curve
    angularBiasOverride=0.999
    angularDampingOverride=1
    angularSpringConstant=200000
    bOverrideAngularBias=true
    bOverrideAngularDamping=true
    alphaBoolBlend {blendOption=Linear}
    boundBone {boneName=dyn_blade_part_2}
    componentAppliedLinearAccClamp=<2000 2000 2000>
    componentLinearAccScale=<0.1 0.1 0.1>
    componentLinearVelScale=<0.1 0.1 0.1>
    physicsBodyDefinitions [{
      boundBone {boneName=dyn_blade_part_2}
      boxExtents=<30 30 30>
      constraintSetup {angularLimitsMax=<0 0 2>, angularLimitsMin=<0 0 -10>, angularTarget=<0 0 0>}
      localJointOffset=<-10 0 0>
     }]
    relativeSpaceBone {boneName=Melee_TwoHanded_Handle}
   }
   n6 = Anim Dynamics - Bone: dyn_blade_part_3!(ComponentPose=@n5, AlphaCurveName:Name=DynamicsAlpha, ExternalForce="0,0,0") -> (Pose) {
    alphaInputType=Curve
    angularBiasOverride=0.999
    angularDampingOverride=1
    angularSpringConstant=200000
    bOverrideAngularBias=true
    bOverrideAngularDamping=true
    simulationSpace=BoneRelative
    alphaBoolBlend {blendOption=Linear}
    boundBone {boneName=dyn_blade_part_3}
    componentAppliedLinearAccClamp=<2000 2000 2000>
    componentLinearAccScale=<0.1 0.1 0.1>
    componentLinearVelScale=<0.1 0.1 0.1>
    physicsBodyDefinitions [{
      boundBone {boneName=dyn_blade_part_3}
      boxExtents=<30 30 30>
      constraintSetup {angularLimitsMax=<0 0 2>, angularLimitsMin=<0 0 -20>, angularTarget=<0 0 0>}
      localJointOffset=<-5 -5 0>
     }]
    relativeSpaceBone {boneName=Melee_TwoHanded_Handle}
   }
   Output Pose(Result=Component To Local(ComponentPose=@n6)) {layerGroup=None}
   AnimGraphNode_StateMachine_0 = Equip State() {is=title AnimGraphNode_StateMachine}
   Save cached pose 'Input pose cache'(Pose=@AnimGraphNode_StateMachine_0) {cacheName=Input pose cache}
   Slot 'DefaultSlot'()
   AnimGraphNode_SubInput_0 = Input Pose()
  ]}
 Equip State {
  entry=Equip
  path=AnimGraph.AnimGraphNode_StateMachine_0.Equip State
  ids=id
  nodes [
   AnimStateNode_0 = Equip() {is=title AnimStateNode}
   AnimStateNode_1 = Unequipped() {is=title AnimStateNode}
   AnimStateTransitionNode_0 = AnimStateTransitionNode()
   AnimStateTransitionNode_5 = AnimStateTransitionNode()
  ]
  links [
   {
    dst=Unequipped
    rule=AnimGraph.AnimGraphNode_StateMachine_0.Equip State.AnimStateTransitionNode_0.Transition
    src=Equip
    via=AnimStateTransitionNode_0
   }
   {
    dst=Equip
    rule=AnimGraph.AnimGraphNode_StateMachine_0.Equip State.AnimStateTransitionNode_5.Transition
    src=Unequipped
    via=AnimStateTransitionNode_5
   }
  ]
 }
 Equip {
  path=AnimGraph.AnimGraphNode_StateMachine_0.Equip State.AnimStateNode_0.Equip
  nodes [Output Animation Pose(Result=Sequence Player 'FreightCalf_Equip_WEP'() {bLoopAnimation=false, sequence=/BRCosmetics/Weapons/FORT_Melee/Pickaxe_FreightCalf/Anim/FreightCalf_Equip_WEP})]
 }
 Unequipped {
  path=AnimGraph.AnimGraphNode_StateMachine_0.Equip State.AnimStateNode_1.Unequipped
  nodes [Output Animation Pose(Result=Local Space Ref Pose())]
 }
 Transition {
  path=AnimGraph.AnimGraphNode_StateMachine_0.Equip State.AnimStateTransitionNode_0.Transition
  nodes [Result(bCanEnterTransition=NOT Boolean(A=Get Equipped() -> (Equipped:bool)))]
 }
 Transition {
  path=AnimGraph.AnimGraphNode_StateMachine_0.Equip State.AnimStateTransitionNode_5.Transition
  nodes [Result(bCanEnterTransition=Get Equipped())]
 }
 EventGraph {nodes [
   Event BlueprintUpdateAnimation()
   > n1 = Branch(Condition=PickaxeItemDefinitionHasAnyMetaTag(Tags=Make Array("[0]"=(TagName="Cosmetics.MetaTags.Style.OptionA")), bRequiresEquipped=true))
   > Set Equipped(Equipped:bool=true) -> (Output_Get:bool)
   Set Equipped(execute=@n1.else, Equipped=false)
  ]}
]
variables [Equipped {type=boolean}]
